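# Animation Blueprint for the MistTycoon mic cosmetic backpack, built on the generic mic stand skeleton.
# The AnimGraph chains two bone-space Transform (Modify) Bone nodes that shift and rotate the offset bone additively, with the second also adding scale.
# Each update, the EventGraph sets SkyDiveFix from the owner's spine-to-pelvis twist alpha multiplied by the skydiving state.
# The AnimGraph shows no use of SkyDiveFix.
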AnimBlueprint /SparksCosmetics/Items/Cosmetics/Backpacks/MistTycoon_Mic/MistTycoon_Mic_Backpack_AnimBP
target_skeleton=/SparksCharacterCommon/Instrument/MicStand/Generic/Meshes/Generic_MicStand_Skeleton
graphs [
 AnimGraph {nodes [
   n1 = "Transform (Modify) Bone - Bone: offset"!(
    ComponentPose=Local To Component(LocalPose=Input Pose())
    bAlphaBoolEnabled:bool=True
    Translation="-2,4,-10"
    Rotation="10,-15,2"
    Scale="1,1,1"
   ) -> (Pose) {
    alphaInputType=Bool
    rotationMode=BMM_Additive
    rotationSpace=BCS_BoneSpace
    scaleMode=BMM_Replace
    translationMode=BMM_Additive
    translationSpace=BCS_BoneSpace
    alphaBoolBlend {blendOption=Linear}
    boneToModify {boneName=offset}
    rotation=<rot 10 -15 2>
    translation=<-2 4 -10>
   }
   n2 = "Transform (Modify) Bone - Bone: offset"!(ComponentPose=@n1, Alpha:float=1, Translation="13,5,2", Rotation="20,0,0", Scale="0.8,0.8,0.8") -> (Pose) {
    rotationMode=BMM_Additive
    rotationSpace=BCS_BoneSpace
    scaleMode=BMM_Additive
    scaleSpace=BCS_BoneSpace
    translationMode=BMM_Additive
    translationSpace=BCS_BoneSpace
    boneToModify {boneName=offset}
    rotation {pitch=20}
    scale=<0.8 0.8 0.8>
    translation=<13 5 2>
   }
   Output Pose(Result=Component To Local(ComponentPose=@n2))
  ]}
 EventGraph {nodes [
   Event BlueprintUpdateAnimation()
   n1 = GetOwnerSkeletalMeshComponent()
   > Set SkyDiveFix(SkyDiveFix:float=float * float(
     A=GetAlphaFromReferenceTransformDelta(SkeletalMeshComponent=@n1, BoneName=spine_05, BaseName=pelvis, InputRange=(X=0,Y=-60), TwistAxis="1,0,0")
     B:bool=Get bIsSkyDiving()
    ) {is=title K2Node_PromotableOperator}) -> (Output_Get:float)
   TryGetPawnOwner()
  ]}
]
variables [SkyDiveFix {type=number}]
